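# Translucent unlit outline material for Niagara mesh particles in the Knight Cat skydiving effect.
# Vertex animation swings each X-side around its own pivot via a dynamic parameter, adds vertex-color-phased sine wobble, and inflates along normals.
# Four offset scene-depth taps detect edges, with spacing scaled by particle camera distance and view size.
# A depth comparison masks out edges over much farther scene pixels; the mobile path outputs zero.
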Material /BRCosmetics/Athena/Effects/Particles/Skydiving/NS_Knight_Cat/M_Knight_Cat_Outline
graph {
 properties {
  bUsedWithNiagaraMeshParticles=true
  blendMode=BLEND_Translucent
  pixelDepthOffsetMode=PDOM_Legacy
  shadingModel=MSM_Unlit
  usedShadingModels=Unlit
 }
 nodes [
  n1 = Transform(None=Constant3Vector() {constant=<rgba 0 1 0 0>}) {transformSourceType=TRANSFORMSOURCE_Local}
  n2 = MaterialFunctionCall(Float3=MaterialFunctionCall() {materialFunction=&LocalPosition}."Local Position (Excluding Offsets)") {materialFunction=&BreakOutFloat3Components}
  n3 = DynamicParameter() {defaultValue=<rgba 3 0 0 1>, paramNames [Power, Sword Derection, AngleFade, None]}
  n4 = Multiply(A=@n3.Sword Derection) {constB=-1}
  n5 = If(A=@n2.R, "A > B"=@n3.Sword Derection, "A == B"=@n4, "A < B"=@n4)
  n6 = Constant3Vector() {constant=<rgba -50 0 0 0>}
  n7 = TransformPosition(None=If(A=@n2.R, "A > B"=Constant3Vector() {constant=<rgba 50 0 0 0>}, "A == B"=@n6, "A < B"=@n6)) {transformType=TRANSFORMPOSSOURCE_World}
  n8 = WorldPosition()
  n9 = Time()
  n10 = Multiply(A=VertexColor().R) {constB=10}
  n11 = Sine(None=Multiply(A=Add(A=@n9, B=@n10)) {constB=0.5})
  n12 = Time()
  n13 = Multiply(A=VertexColor().R) {constB=10}
  n14 = Sine(None=Multiply(A=Add(A=@n12, B=@n13)) {constB=0.5})
  n15 = Add(
   A=LinearInterpolate(A=Constant3Vector() {constant=<rgba 0 0 10 0>}, B=Constant3Vector() {constant=<rgba 0 0 -10 0>}, Alpha=ConstantBiasScale(None=@n11))
   B=LinearInterpolate(A=Constant3Vector() {constant=<rgba 0 7 0 0>}, B=Constant3Vector() {constant=<rgba 0 -7 0 0>}, Alpha=ConstantBiasScale(None=@n14))
  ) {constB=0}
  Add(
   A=Add(
    A=RotateAboutAxis(NormalizedRotationAxis=@n1, RotationAngle=@n5, PivotPoint=@n7, Position=@n8.XYZ)
    B=Transform(None=@n15) {transformSourceType=TRANSFORMSOURCE_Local}
   )
   B=Multiply(A=VertexNormalWS()) {constB=2}
  )
  Constant() {r=1}
  n16 = ComponentMask(None=SceneTexture() {sceneTextureId=PPI_SceneDepth}.Color) {r=true}
  n17 = Constant() {r=1}
  n18 = If(A=@n16, B=Multiply(A=Constant() {r=1.5}, B=PixelDepth().R), "A > B"=Constant(), "A == B"=@n17, "A < B"=@n17)
  n19 = ScreenPosition()
  n20 = Distance(A=ParticlePositionWS(), B=CameraPositionWS())
  n21 = Multiply(A=Divide(B=Multiply(A=@n20) {constB=0.015}), B=ScalarParameter() {defaultValue=6, parameterName=sDT2})
  n22 = Subtract(A=Divide(A=ViewSize(), B=Constant() {r=320}), B=Constant() {r=2}) {constB=0}
  n23 = Multiply(A=Clamp(None=Ceil(None=@n22)) {maxDefault=2, minDefault=1}, B=SceneTexelSize()) {constA=2}
  n24 = Constant()
  n25 = AppendVector(A=ComponentMask(None=@n23) {desc=x, r=true}, B=@n24)
  n26 = SceneTexture(UVs=Add(A=@n19.ViewportUV, B=Multiply(A=@n25, B=@n21)) {desc=x+}) {sceneTextureId=PPI_SceneDepth}
  n27 = AppendVector(A=@n24, B=ComponentMask(None=@n23) {desc=y, g=true}) {desc=y+}
  n28 = SceneTexture(UVs=Add(A=@n19.ViewportUV, B=Multiply(A=@n27, B=@n21)) {desc=y+}) {sceneTextureId=PPI_SceneDepth}
  n29 = Max(A=Subtract(A=ComponentMask(None=@n26.Color) {r=true}, B=@n16), B=Subtract(A=ComponentMask(None=@n28.Color) {r=true}, B=@n16))
  n30 = Constant() {r=-1}
  n31 = Add(A=@n19.ViewportUV, B=Multiply(A=Multiply(A=@n25, B=@n30), B=@n21)) {desc=x-}
  n32 = Subtract(A=ComponentMask(None=SceneTexture(UVs=@n31) {sceneTextureId=PPI_SceneDepth}.Color) {r=true}, B=@n16)
  n33 = Multiply(A=Multiply(A=@n30, B=@n27), B=@n21)
  n34 = ComponentMask(None=SceneTexture(UVs=Add(A=@n19.ViewportUV, B=@n33) {desc=y-}) {sceneTextureId=PPI_SceneDepth}.Color) {r=true}
  n35 = Constant()
  n36 = If(
   A=Max(A=Max(A=@n29, B=@n32), B=Subtract(A=@n34, B=@n16))
   B=ScalarParameter() {defaultValue=5, parameterName=DT}
   "A > B"=Constant() {r=1}
   "A == B"=@n35
   "A < B"=@n35
  )
  ShadingPathSwitch(Default=LinearInterpolate(A=Constant(), B=Constant() {r=1}, Alpha=Multiply(A=@n18, B=@n36)), Mobile=Constant()) {desc=Migrated from FeatureLevelSwitch}
 ]
}
