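# Masked, lit material for a pink Knight Cat sword in a skydiving Niagara effect, reusing the Knight Cat pickaxe textures.
# Emissive texture is tinted magenta or swapped for particle color, with an optional radial glow mask faded by view angle and particle alpha.
# An offset chain rotates each local-X side around mirrored pivots via a dynamic parameter, adding sine wobble phased by vertex red.
# A curve-atlas gradient with Fresnel forms another color branch.
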Material /BRCosmetics/Athena/Effects/Particles/Skydiving/NS_Knight_Cat/M_Knight_Cat_Sword_Pink
graph {
 properties {
  bUsedWithMeshDeformer=true
  bUsedWithNiagaraMeshParticles=true
  bUsedWithNiagaraSprites=true
  bUsedWithSkeletalMesh=true
  blendMode=BLEND_Masked
  pixelDepthOffsetMode=PDOM_Legacy
  usedShadingModels=Default Lit
 }
 nodes [
  Constant() {r=1}
  n1 = TextureSampleParameter2D() {parameterName=Emissive_Tex, texture=/BRCosmetics/Weapons/FORT_Melee/Pickaxe_Knight_Cat_Female/Textures/T_Pickaxe_Knight_Cat_Female_E}
  n2 = LinearInterpolate(
   A=Multiply(A=@n1.RGB, B=VectorParameter() {parameterName=Emissive_Color, defaultValue=<rgba 3 0 2.5742 1>}.RGB)
   B=ParticleColor().RGB
   Alpha=ScalarParameter() {parameterName=Use Only simple glow}
  )
  LinearInterpolate(A=@n2, Alpha=ScalarParameter() {parameterName=Use Only Simple Square})
  MaterialFunctionCall(Emissive=@n2) {materialFunction=&MF_FX_EmissiveAdjustments}
  n3 = DynamicParameter() {defaultValue=<rgba 3 0 0 1>, paramNames [Power, Sword Derection, AngleFade, None]}
  n4 = Power(
   Base=SphereMask(
    A=TextureCoordinate()
    B=Constant() {r=0.5}
    Radius=ScalarParameter() {defaultValue=0.5, parameterName=Radius}
    Hardness=ScalarParameter() {parameterName=Hardness}
   ) {hardnessPercent=0}
   Exp=@n3.Power
  ) {constExponent=3}
  n5 = Multiply(
   A=ParticleColor().A
   B=MaterialFunctionCall("FadeSensitivity (optional)"=@n3.AngleFade) {materialFunction=&FadeBasedOnViewAngle}.Result
  )
  Multiply(
   A=LinearInterpolate(A=Constant() {r=1}, B=Multiply(A=@n4, B=@n5), Alpha=ScalarParameter() {parameterName=Use Only simple glow})
   B=ParticleColor().A
  )
  n6 = ComponentMask(None=TextureCoordinate() {coordinateIndex=1}) {g=true}
  n7 = CurveAtlasRowParameter(CurveTime=Saturate(None=Power(Base=@n6))) {
   atlas=/Game/Athena/LinearColorCurves/A_Atlas_Turtleneck_Glider
   curve=/Game/Athena/LinearColorCurves/Cube_Knight_Cat_GreenCurve
   defaultValue=5
  }
  n8 = Power(Base=Fresnel() {baseReflectFraction=0, exponent=1}) {constExponent=1}
  Multiply(A=Multiply(A=@n7, B=LinearInterpolate(Alpha=@n8) {constA=0.4})) {constB=2}
  n9 = Constant()
  Reroute(None=Multiply(A=Desaturation(Fraction=@n9)) {constB=0.9})
  n10 = Transform(None=Constant3Vector() {constant=<rgba 0 1 0 0>}) {transformSourceType=TRANSFORMSOURCE_Local}
  n11 = MaterialFunctionCall(Float3=MaterialFunctionCall() {materialFunction=&LocalPosition}."Local Position (Excluding Offsets)") {materialFunction=&BreakOutFloat3Components}
  n12 = DynamicParameter() {defaultValue=<rgba 3 0 0 1>, paramNames [Power, Sword Derection, AngleFade, None]}
  n13 = Multiply(A=@n12.Sword Derection) {constB=-1}
  n14 = If(A=@n11.R, "A > B"=@n12.Sword Derection, "A == B"=@n13, "A < B"=@n13)
  n15 = Constant3Vector() {constant=<rgba -50 0 0 0>}
  n16 = TransformPosition(None=If(A=@n11.R, "A > B"=Constant3Vector() {constant=<rgba 50 0 0 0>}, "A == B"=@n15, "A < B"=@n15)) {transformType=TRANSFORMPOSSOURCE_World}
  n17 = WorldPosition()
  n18 = Time()
  n19 = Multiply(A=VertexColor().R) {constB=10}
  n20 = Sine(None=Multiply(A=Add(A=@n18, B=@n19)) {constB=0.5})
  n21 = Time()
  n22 = Multiply(A=VertexColor().R) {constB=10}
  n23 = Sine(None=Multiply(A=Add(A=@n21, B=@n22)) {constB=0.5})
  n24 = Add(
   A=LinearInterpolate(A=Constant3Vector() {constant=<rgba 0 0 10 0>}, B=Constant3Vector() {constant=<rgba 0 0 -10 0>}, Alpha=ConstantBiasScale(None=@n20))
   B=LinearInterpolate(A=Constant3Vector() {constant=<rgba 0 7 0 0>}, B=Constant3Vector() {constant=<rgba 0 -7 0 0>}, Alpha=ConstantBiasScale(None=@n23))
  ) {constB=0}
  Reroute(None=Add(
    A=RotateAboutAxis(NormalizedRotationAxis=@n10, RotationAngle=@n14, PivotPoint=@n16, Position=@n17.XYZ)
    B=Transform(None=@n24) {transformSourceType=TRANSFORMSOURCE_Local}
   ))
  Reroute(None=@n2)
  ScalarParameter() {defaultValue=0.05, parameterName=sss}
  TextureSampleParameter2D() {parameterName=Default_Tex, texture=/BRCosmetics/Weapons/FORT_Melee/Pickaxe_Knight_Cat_Female/Textures/T_Pickaxe_Knight_Cat_Female_D}
  TextureSampleParameter2D() {
   parameterName=Specular_Tex
   samplerType=SAMPLERTYPE_Masks
   texture=/BRCosmetics/Weapons/FORT_Melee/Pickaxe_Knight_Cat_Female/Textures/T_Pickaxe_Knight_Cat_Female_S
  }
  TextureSample() {samplerType=SAMPLERTYPE_Normal, texture=/BRCosmetics/Weapons/FORT_Melee/Pickaxe_Knight_Cat_Female/Textures/T_Pickaxe_Knight_Cat_Female_N}
  TextureSample() {samplerType=SAMPLERTYPE_Masks, texture=&T_Pickaxe_Knight_Cat_Female_S}
  TextureSample() {samplerType=SAMPLERTYPE_Normal, texture=&T_Pickaxe_Knight_Cat_Female_N}
 ]
}
