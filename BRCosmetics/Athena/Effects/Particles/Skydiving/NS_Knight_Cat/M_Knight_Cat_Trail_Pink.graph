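# Pink translucent, unlit, two-sided trail for the Knight Cat skydiving effect, built for Niagara ribbons and mesh particles.
# Emissive adds a thin colored hard line to a curve-atlas gradient, tinted by particle color.
# Opacity combines opposing panned smoke and blurred noise with UV edge fades and particle alpha.
# World position offset layers sine waves, a clamped bend from Niagara dynamic parameters relative to particle position, and a side-dependent rotation about a local axis.
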Material /BRCosmetics/Athena/Effects/Particles/Skydiving/NS_Knight_Cat/M_Knight_Cat_Trail_Pink
graph {
 properties {
  bUsedWithNiagaraMeshParticles=true
  bUsedWithNiagaraRibbons=true
  blendMode=BLEND_Translucent
  pixelDepthOffsetMode=PDOM_Legacy
  shadingModel=MSM_Unlit
  twoSided=true
  usedShadingModels=Unlit
 }
 nodes [
  n1 = CurveAtlasRowParameter(CurveTime=ComponentMask(None=TextureCoordinate()) {r=true}) {
   atlas=/Game/Athena/LinearColorCurves/A_Atlas_Turtleneck_Glider
   curve=/Game/Athena/LinearColorCurves/Cube_Knight_Cat_PinkCuve
   defaultValue=4
   parameterName=Curve02
  }
  LinearInterpolate(A=Constant() {r=1}, B=@n1, Alpha=ScalarParameter() {parameterName=Use Purple trail})
  n2 = TextureCoordinate()
  n3 = OneMinus(None=Sine(None=ComponentMask(None=@n2) {g=true}) {period=2})
  n4 = Step(Y=ComponentMask(None=TextureCoordinate()) {g=true}) {constX=0.5}
  n5 = Round(None=Multiply(A=Power(Base=@n3) {constExponent=10}, B=Saturate(None=@n4)))
  n6 = ComponentMask(None=TextureCoordinate()) {r=true}
  n7 = Saturate(None=Multiply(A=Power(Base=@n6) {constExponent=2.5}) {constB=5})
  n8 = ComponentMask(None=TextureCoordinate()) {r=true}
  n9 = Multiply(A=Power(Base=OneMinus(None=@n8)) {constExponent=2.5}) {constB=5}
  n10 = Multiply(
   A=VectorParameter() {parameterName=Hard_Line_Color, defaultValue=<rgba 10 0 5 1>}.RGB
   B=Multiply(A=Multiply(A=@n5, B=@n7), B=Saturate(None=@n9))
  )
  n11 = TextureCoordinate()
  n12 = CurveAtlasRowParameter(CurveTime=OneMinus(None=ComponentMask(None=@n11) {g=true})) {
   atlas=&A_Atlas_Turtleneck_Glider
   curve=&Cube_Knight_Cat_PinkCuve
   defaultValue=4
   parameterName=Curve01
  }
  n13 = Multiply(A=@n12)
  n14 = ParticleColor()
  MaterialFunctionCall(Emissive=Multiply(A=Add(A=@n10, B=@n13), B=@n14.RGB)) {materialFunction=&MF_FX_EmissiveAdjustments}
  n15 = TextureCoordinate() {uTiling=0.05, vTiling=0.8}
  n16 = AppendVector(A=Add(A=ComponentMask(None=@n15) {r=true}, B=VertexColor().R), B=ComponentMask(None=@n15) {g=true})
  n17 = Power(Base=TextureSample(UVs=Panner(Coordinate=@n16) {speedX=-0.5}) {texture=&T_Inky_Smoke_Tile}.R) {constExponent=5}
  n18 = ScalarParameter() {defaultValue=2, parameterName=Alpha_Line}
  n19 = Multiply(A=ComponentMask(None=TextureCoordinate()) {g=true}) {constB=0.2}
  n20 = TextureCoordinate() {uTiling=0.05, vTiling=0.8}
  n21 = Add(A=ComponentMask(None=@n20) {r=true}, B=VertexColor().R)
  n22 = ComponentMask(None=@n20) {g=true}
  n23 = Panner(Coordinate=OneMinus(None=AppendVector(A=@n21, B=@n22))) {speedX=0.5}
  n24 = Multiply(
   A=Power(Base=TextureSample(UVs=@n23) {samplerType=SAMPLERTYPE_Grayscale, texture=&LowResBlurredNoise}.R) {constExponent=5}
   B=ScalarParameter() {defaultValue=60, parameterName=Alpha_Line2}
  ) {constB=60}
  n25 = OneMinus(None=ComponentMask(None=TextureCoordinate()) {g=true})
  n26 = Saturate(None=Power(Base=@n25) {constExponent=3})
  n27 = TextureSample(UVs=TextureCoordinate() {uTiling=0, vTiling=0.5}) {samplerType=SAMPLERTYPE_Grayscale, texture=&LowResBlurredNoise}
  n28 = Multiply(A=Power(Base=@n25) {constExponent=3}) {constB=4.5}
  n29 = TextureCoordinate()
  n30 = Multiply(
   A=LinearInterpolate(A=Saturate(None=@n24), B=Multiply(A=@n26, B=@n27.R) {constB=0.5}, Alpha=Saturate(None=@n28))
   B=OneMinus(None=ComponentMask(None=@n29) {g=true})
  )
  n31 = Multiply(A=Add(A=Multiply(A=@n17, B=@n18) {constB=2}, B=Add(A=@n19, B=@n30)))
  n32 = TextureCoordinate()
  n33 = Power(Base=OneMinus(None=ComponentMask(None=@n32) {r=true})) {constExponent=1.5}
  n34 = Power(Base=ComponentMask(None=TextureCoordinate()) {r=true}) {constExponent=0.9}
  n35 = Multiply(A=Saturate(None=Multiply(A=@n33) {constB=3}), B=Saturate(None=Multiply(A=@n34) {constB=6}))
  n36 = TextureCoordinate()
  n37 = Multiply(A=OneMinus(None=ComponentMask(None=@n36) {g=true})) {constB=10}
  Multiply(A=@n14.A, B=Multiply(A=Multiply(A=@n31, B=@n35), B=Saturate(None=@n37)))
  n38 = Transform(None=Constant3Vector() {constant=<rgba 0 1 0 0>}) {transformSourceType=TRANSFORMSOURCE_Local}
  n39 = MaterialFunctionCall(Float3=MaterialFunctionCall() {materialFunction=&LocalPosition}."Local Position (Excluding Offsets)") {materialFunction=&BreakOutFloat3Components}
  n40 = DynamicParameter() {defaultValue=<rgba 3 0 0 1>, paramNames [Power, Sword Derection, AngleFade, None]}
  n41 = Multiply(A=@n40.Sword Derection) {constB=-1}
  n42 = If(A=@n39.R, "A > B"=@n40.Sword Derection, "A == B"=@n41, "A < B"=@n41)
  n43 = Constant3Vector() {constant=<rgba -50 0 0 0>}
  n44 = TransformPosition(None=If(A=@n39.R, "A > B"=Constant3Vector() {constant=<rgba 50 0 0 0>}, "A == B"=@n43, "A < B"=@n43)) {transformType=TRANSFORMPOSSOURCE_World}
  n45 = WorldPosition()
  n46 = Constant3Vector() {constant=<rgba 1 0 0 0>}
  n47 = Time()
  n48 = Add(A=TextureCoordinate() {uTiling=0.8, vTiling=0}, B=VertexColor().R)
  n49 = Sine(None=Add(A=Multiply(A=@n47) {constB=-0.8}, B=ComponentMask(None=@n48) {r=true}))
  n50 = Power(Base=ComponentMask(None=TextureCoordinate()) {r=true}) {constExponent=1}
  n51 = Clamp(None=LinearInterpolate(B=Multiply(A=@n46, B=@n49), Alpha=Saturate(None=@n50))) {maxDefault=5, minDefault=-0.4}
  n52 = Constant3Vector() {constant=<rgba 0 0 10 0>}
  n53 = Constant3Vector() {constant=<rgba 0 0 -10 0>}
  n54 = Add(A=Time(), B=Multiply(A=VertexColor().R) {constB=10})
  n55 = ConstantBiasScale(None=Sine(None=Multiply(A=@n54) {constB=0.5}))
  n56 = Constant3Vector() {constant=<rgba 0 7 0 0>}
  n57 = Constant3Vector() {constant=<rgba 0 -7 0 0>}
  n58 = Add(A=Time(), B=Multiply(A=VertexColor().R) {constB=10})
  n59 = ConstantBiasScale(None=Sine(None=Multiply(A=@n58) {constB=0.5}))
  n60 = Add(B=Add(A=LinearInterpolate(A=@n52, B=@n53, Alpha=@n55), B=LinearInterpolate(A=@n56, B=@n57, Alpha=@n59)) {constB=0})
  n61 = DynamicParameter() {desc=Parameters set by Niagara, parameterIndex=1, paramNames [X,Y,Z,WPOIntensity]}
  n62 = Subtract(
   A=MaterialFunctionCall(X=@n61.X, Y=@n61.X, Z=@n61.X) {materialFunction=&MakeFloat3}.Result
   B=ParticlePositionWS()
  )
  n63 = Clamp(
   None=ComponentMask(None=@n62) {g=true, r=true}
   Min=ScalarParameter() {defaultValue=-100, parameterName="X\\YClamp_min"}
   Max=ScalarParameter() {defaultValue=100, parameterName="X\\YClamp_max"}
  ) {maxDefault=1000, minDefault=-1000}
  n64 = Multiply(
   A=Clamp(
    None=ComponentMask(None=@n62) {b=true}
    Min=ScalarParameter() {defaultValue=-30, parameterName=Z Clamp_min}
    Max=ScalarParameter() {defaultValue=30, parameterName=Z Clamp_max}
   ) {maxDefault=1000, minDefault=-1000}
   B=DynamicParameter() {paramNames [Z Length, Param2, Param3, Param4]}.Z Length
  )
  n65 = TextureCoordinate()
  n66 = Multiply(A=Power(Base=ComponentMask(None=@n65) {r=true}, Exp=ScalarParameter() {defaultValue=1.5, parameterName=WPOBendPower}), B=@n61.WPOIntensity)
  n67 = LinearInterpolate(
   A=Constant()
   B=LinearInterpolate(B=AppendVector(A=@n63, B=@n64), Alpha=LinearInterpolate(Alpha=@n66) {constB=2})
   Alpha=ScalarParameter() {defaultValue=1, parameterName=Use interpolate WPO}
  )
  n68 = Add(
   A=Transform(None=Multiply(A=@n51) {constB=6}) {transformSourceType=TRANSFORMSOURCE_Local}
   B=Transform(None=Add(A=@n60, B=@n67)) {transformSourceType=TRANSFORMSOURCE_Local}
  )
  n69 = ScalarParameter() {defaultValue=1, parameterName=Use_WPO}
  Reroute(None=Add(
    A=LinearInterpolate(B=@n68, Alpha=@n69)
    B=RotateAboutAxis(NormalizedRotationAxis=@n38, RotationAngle=@n42, PivotPoint=@n44, Position=@n45.XYZ)
   ))
  ScalarParameter() {defaultValue=0.05, parameterName=sss}
 ]
}
